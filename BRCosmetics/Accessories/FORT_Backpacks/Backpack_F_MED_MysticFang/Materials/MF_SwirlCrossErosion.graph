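# Material function for the Mystic Fang backpack that builds an animated liquid swirl mask from a grayscale texture.
# Two half-cycle-offset layers scroll vertically, scale around a pivot between min and max in log space, and use BumpOffset for parallax depth.
# Each layer erodes through a time-driven threshold shaped by Swirl Power, peaking at its loop point.
# A final lerp blends both layers into one looping output.
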MaterialFunction /BRCosmetics/Accessories/FORT_Backpacks/Backpack_F_MED_MysticFang/Materials/MF_SwirlCrossErosion
graph {nodes [
  n1 = Constant2Vector() {g=0.5, r=0.5}
  n2 = Subtract(A=TextureCoordinate(), B=Subtract(A=FunctionInput(Preview=@n1) {inputName=Scale Pivot, inputType=FunctionInput_Vector2}) {constB=0.5})
  n3 = Constant() {r=10}
  n4 = Logarithm10!(X=FunctionInput(Preview=Constant() {r=0.5}) {inputName=Min Scale, inputType=FunctionInput_Scalar})
  n5 = Logarithm10(X=FunctionInput(Preview=Constant() {r=2}) {inputName=Max Scale, inputType=FunctionInput_Scalar})
  n6 = Power(Base=@n3, Exp=LinearInterpolate(A=@n4, B=@n5, Alpha=NamedRerouteUsage()))
  n7 = Multiply(A=Time(), B=Constant2Vector() {g=0.15})
  n8 = Multiply(A=FunctionInput() {inputName=Depth, inputType=FunctionInput_Scalar}) {constB=-1}
  n9 = NamedRerouteUsage()
  n10 = BumpOffset(
   Coordinate=Add(
    A=MaterialFunctionCall(UVs=@n2, Texture Scale=@n6) {materialFunction=&ScaleUVsByCenter}.UVs
    B=@n7
   )
   Height=Multiply(A=@n8, B=OneMinus(None=@n9)) {constB=-5}
  )
  n11 = TextureObjectParameter() {
   group=Liquid - Swirl
   parameterName=Swirl Texture
   samplerType=SAMPLERTYPE_LinearGrayscale
   sortPriority=0
   texture=/Game/Global/Textures/Default/Blanks/T_BlackLinearGrayscale
  }
  n12 = Constant() {r=5}
  n13 = Divide(
   A=Time()
   B=NamedRerouteDeclaration(None=FunctionInput(Preview=@n12) {inputName=Duration, inputType=FunctionInput_Scalar}) {name=FadeDuration, nodeColor=<rgba 1 0 0.5156 1>}
  )
  n14 = ConstantBiasScale(None=NamedRerouteDeclaration(None=Frac(None=@n13)) {name=Layer1_alpha, nodeColor=<rgba 0.3828 1 0 1>}) {bias=-0.5, scale=2}
  n15 = ScalarParameter() {defaultValue=2, group=Liquid - Swirl, parameterName=Swirl Power, sortPriority=1}
  n16 = Power(Base=NamedRerouteDeclaration(None=Abs(None=@n14)) {name=CrossFade_Alpha, nodeColor=<rgba 0 0.8359 1 1>}, Exp=@n15)
  n17 = Divide(
   A=Subtract(
    A=TextureSample(UVs=@n10, Tex=@n11) {samplerType=SAMPLERTYPE_LinearGrayscale, texture=&T_BlackLinearGrayscale}.R
    B=@n16
   )
   B=Subtract(B=@n16)
  )
  n18 = Power(Base=@n3, Exp=LinearInterpolate(A=@n4, B=@n5, Alpha=NamedRerouteUsage()))
  n19 = NamedRerouteUsage()
  n20 = BumpOffset(
   Coordinate=Add(
    A=MaterialFunctionCall(UVs=@n2, Texture Scale=@n18) {materialFunction=&ScaleUVsByCenter}.UVs
    B=@n7
   )
   Height=Multiply(A=OneMinus(None=@n19), B=@n8) {constB=-5}
  )
  n21 = NamedRerouteDeclaration(None=Frac(None=Add(A=@n13) {constB=0.5})) {name=Layer2_alpha, nodeColor=<rgba 1 0 0.7031 1>}
  n22 = Power(Base=Abs(None=ConstantBiasScale(None=@n21) {bias=-0.5, scale=2}), Exp=@n15)
  n23 = Divide(
   A=Subtract(
    A=TextureSample(UVs=@n20, Tex=@n11) {samplerType=SAMPLERTYPE_LinearGrayscale, texture=&T_BlackLinearGrayscale}.R
    B=@n22
   )
   B=Subtract(B=@n22)
  )
  FunctionOutput(None=LinearInterpolate(A=Saturate(None=@n17), B=Saturate(None=@n23), Alpha=NamedRerouteUsage()))
 ]}
