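# Material function in the Watt Superlevels skin's parent materials that crops and repositions a texture between two UV corners, useful for emissives.
# It remaps UVs to the corner rectangle and samples the texture with a shared wrap sampler, although its input description requests Clamp.
# Two per-axis SphereMasks multiply into a rectangular mask.
# It outputs the sample (static-switched between RGB and an appended vector), masked sample, crop UVs and mask.
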MaterialFunction /BRCosmetics/Characters/Player/Female/Medium/Bodies/F_MED_SpatialTravel_Watt/Skins/Superlevels/ParentMaterials/MF_TextureCropping_SL
graph {
 properties {
  bExposeToLibrary=true
  description="Crops a texture, useful for Emissives.  Supply the top left and bottom right corners of the crop  in 0-1 range."
  libraryCategoriesText [Misc,Texturing]
 }
 nodes [
  n1 = FunctionInput(Preview=TextureCoordinate()) {bUsePreviewValueAsDefault=true, description=Default value is texture coordinate 0., inputName=UVs, inputType=FunctionInput_Vector2}
  n2 = Constant() {r=-1}
  n3 = FunctionInput() {
   bUsePreviewValueAsDefault=true
   description="New UV location for UV value (0,0)."
   inputName=UpperLeftCorner
   inputType=FunctionInput_Vector2
   sortPriority=1
  }
  n4 = FunctionInput() {
   bUsePreviewValueAsDefault=true
   description="New location for uv location (1,1)."
   inputName=LowerRightCorner
   inputType=FunctionInput_Vector2
   sortPriority=2
   previewValue=<1 1 0 1>
  }
  n5 = Divide(A=Add(A=@n1, B=Multiply(A=@n2, B=@n3)), B=Subtract(A=@n4, B=@n3))
  n6 = FunctionInput(Preview=TextureObject() {texture=/Engine/EngineMaterials/DefaultDiffuse}) {
   bUsePreviewValueAsDefault=true
   description="Texture to be positioned and cropped, make sure it's Address is set to Clamp"
   inputName=TextureIn
   inputType=FunctionInput_Texture2D
  }
  n7 = TextureSample(UVs=@n5, Tex=@n6) {samplerSource=SSM_Wrap_WorldGroupSettings}
  n8 = FunctionInput(Preview=StaticBool()) {
   bUsePreviewValueAsDefault=true
   description=False will export an RGB result and True will export a RGBA result.
   inputName=ExportFloat4
   inputType=FunctionInput_StaticBool
   sortPriority=5
  }
  n9 = StaticSwitch(True=AppendVector(A=@n7.RGB, B=@n7.RGB), False=@n7.RGB, Value=@n8)
  FunctionOutput(None=@n9) {description=7 Instructions with parameters 2 without, outputName=Cropped}
  n10 = Constant() {r=0.5}
  n11 = Multiply(A=SphereMask(A=ComponentMask(None=@n5) {r=true}, B=@n10, Radius=@n10), B=SphereMask(A=ComponentMask(None=@n5) {g=true}, B=@n10, Radius=@n10))
  FunctionOutput(None=Multiply(A=@n9, B=@n11)) {description=11 instructions with parameter 7 without, outputName=CroppedMasked, sortPriority=1}
  FunctionOutput(None=@n5) {description=7 Instructions with parameters 2 without, outputName=Crop UVs, sortPriority=3}
  FunctionOutput(None=@n11) {description=11 instructions with parameter 7 without, outputName=Crop Mask, sortPriority=4}
 ]
}
